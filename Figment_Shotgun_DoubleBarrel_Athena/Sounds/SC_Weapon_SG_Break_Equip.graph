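# Sound cue for the equip sound of the double-barrel break-action shotgun, played as weapon foley.
# It plays a reused break-close reload wave through a modulator that randomizes pitch upward, then a short fixed delay.
# Reload attenuation shapes its falloff, and a concurrency override limits it to two instances per owner, stopping the oldest.
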SoundCue /Figment_Shotgun_DoubleBarrel_Athena/Sounds/SC_Weapon_SG_Break_Equip
root=Delay_0
nodes [
 {id=Delay_0, children [Modulator_0], properties {delayMax=0.1, delayMin=0.1}}
 {id=Modulator_0, children [WavePlayer_11], properties {pitchMax=1.4, pitchMin=1.2}}
 {id=WavePlayer_11, properties {soundWaveAssetPtr=/Game/Sounds/Fort_Weapons/Shotgun/Reload_Elements/weapon_shotgun_break_close_01}}
]
properties {
 attenuationSettings=&Fort_Weapon_Reload_Att
 bOverrideConcurrency=true
 maxDistance=2250
 soundClassObject=&Fort_WeaponFoley
 volumeMultiplier=0.601
 concurrencyOverrides {bLimitToOwner=true, maxCount=2, resolutionRule=StopOldest}
}
